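# Unlit glow material for the GameInventoryToy trap item in Athena, also enabled for spline meshes.
# It blends a ColorMain parameter (cyan-blue by default) toward ColorOutline (pale green by default) using the vertex color red channel.
# The main and outline regions come from mesh vertex paint, with no textures in the graph.
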Material /Game/Athena/Items/Traps/GameInventoryToy/M_StartingInventory_Glow
graph {
 properties {bUsedWithSplineMeshes=true, pixelDepthOffsetMode=PDOM_Legacy, shadingModel=MSM_Unlit, usedShadingModels=Unlit}
 nodes [LinearInterpolate(
   A=VectorParameter() {parameterName=ColorMain, defaultValue=<rgba 0.0538 0.6185 0.8385 1>}.RGB
   B=VectorParameter() {parameterName=ColorOutline, defaultValue=<rgba 0.4444 1 0.4482 1>}.RGB
   Alpha=VertexColor().R
  )]
}
