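; Creative floor device that grants a starting inventory to players of a configured team, or all teams, and applies team and weapon settings to the minigame.
; Pickups hitting its drop box fill an item list, granted on pawn possession, teleport or team switch, with optional extra ammo and auto-equip.
; It broadcasts elimination and out-of-respawns events, and can end the round from an event.
; Team-colored materials and a preview component display its state.
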(event EditorPreview|EventSetupForEditorPreview
  ;[0.08761 1 0 0.5] Initialization
  (CallFunction|PropertiesUpdated)
  (Class|CreativeItemPreviewComponent|EditorDisplayPreview (Variables|Default|GetCreative_ItemPreview_Component) (Variables|Default|GetPreviewStaticMesh)))

(event EventOnWorldReady
  ;[0.08761 1 0 0.5] Initialization
  (CallFunction|OverrideCollisionSettings))

(event EventOnPlaced
  ;[0.08761 1 0 0.5] Initialization
  (Audio|PlaySoundatLocation (Variables|Audio|GetPlaced%20Sound) (Transformation|GetActorLocation)))

(event Custom|Setup%20Respawn%20Item%20Grant%20Binds
  (params (Minigame :type Fort%20Minigame%20Object%20Reference))
  ;[0.4374 0.9901 1] When player pawn is possessed give the player the items
  (Networking|SwitchHasAuthority
    (:Authority
      (Utilities|IsValid Minigame
        (:Is%20Valid
          (for _array_element (Minigame|GetParticipatingPlayers Minigame)
            (Utilities|IsValid _array_element
              (:Is%20Valid
                (Utilities|IsValid (Fortnite|GetControllerfromActor _array_element)
                  (:Is%20Valid
                    (Pawn|BindEventtoOnPossessedPawnChanged (Fortnite|GetControllerfromActor _array_element) (delegate Custom|OnPossessedPawnChanged)))))))
          (Minigame|BindEventtoPlayerAddedDelegate Minigame (delegate Custom|OnPlayerAddedToMinigame)))))))

(event EventBeginPlay
  ;[0.08761 1 0 0.5] Initialization
  (|CallParentFunction)
  (Networking|SwitchHasAuthority
    (:Authority
      (CallFunction|MigrateDatatoPickupList)
      (Default|BindEventtoOnItemAdded (Variables|Default|GetPickupItemList) (EventDispatchers|CreateEvent :!event Custom|OnAddedToInventoryList)))
    (:Remote
      (Initialization|InitializeComponent (Variables|Default|GetCreative_ItemPreview_Component) (Variables|Default|GetRotationSceneComponent) (Variables|Default|GetPreviewStaticMesh) (Variables|Default|GetPreviewSkeletalMesh) (Variables|Default|GetHolo%20Loop%20Cue) (Variables|Default|GetPickupItemList) (Variables|Default|GetItemVariantNotifier) :!class /Game/Creative/Devices/Common/Components/Creative_ItemPreview_Component.Creative_ItemPreview_Component_C)))
  (Default|AssignOnItemListUpdated (Variables|Default|GetPickupItemList) (EventDispatchers|CreateEvent :!event Custom|OnItemListUpdated)))

(event Custom|OnPlayerAddedToMinigame
  (params (UniqueNetId :type Unique%20Net%20Id%20Structure) (bIsLocalPlayer :type Boolean))
  (bind _returnvalue (Fortnite|Athena|GameState|GetGameStateAthena))
  (bind _returnvalue_2 (Class|FortGameState|GetFortPlayerStatefromUniqueNetId _returnvalue UniqueNetId))
  (bind _returnvalue_3 (Fortnite|GetControllerfromActor _returnvalue_2))
  ;[0.4374 0.9901 1] When player pawn is possessed give the player the items
  (Utilities|IsValid _returnvalue
    (:Is%20Valid
      (Utilities|IsValid _returnvalue_2
        (:Is%20Valid
          (Utilities|IsValid _returnvalue_3
            (:Is%20Valid
              (Pawn|BindEventtoOnPossessedPawnChanged _returnvalue_3 (delegate Custom|OnPossessedPawnChanged)))))))))

(event Custom|OnPossessedPawnChanged
  (params (OldPawn :type Pawn%20Object%20Reference) (NewPawn :type Pawn%20Object%20Reference))
  ;[0.4374 0.9901 1] When player pawn is possessed give the player the items
  (bind _asfort_player_pawn (Utilities|Casting|CastToFortPlayerPawn NewPawn))
  (bind _currentminigame (Variables|/Script/FortniteGame.FortMinigameLogicComponent|GetCurrentMinigame (Variables|Default|GetFortMinigameProgress))
    (:then
      (bind _gameisrunning (Default|MinigameIsRunning))
      (if (and _gameisrunning (Minigame|IsinProgress _currentminigame))
        (EventDispatchers|BindEventtoOnDied _asfort_player_pawn (delegate Custom|OnPlayerPawnDied) :!class /Script/FortniteGame.FortPawn)
        (bind _asfort_player_state_athena (Utilities|Casting|CastToFortPlayerStateAthena (Variables|/Script/Engine.Controller|GetPlayerState (Pawn|GetController _asfort_player_pawn))))
        (if (CallFunction|IsActoronSameTeam _asfort_player_pawn)
          (Utilities|ShouldGrantItemstoPlayer (PlayerState|GetPlayerUniqueNetID _asfort_player_state_athena) (Pawn|GetController _asfort_player_pawn)
            (:true
              (bind _asfort_inventory_owner_interface (Utilities|Casting|CastToFortInventoryOwnerInterface (Pawn|GetController _asfort_player_pawn)))
              (CallFunction|TryGiveAllItemstoPlayer _asfort_inventory_owner_interface)
              (Utilities|Set|Add (Variables|Default|GetGrantedPlayers) (PlayerState|GetPlayerUniqueNetID _asfort_player_state_athena))))
          (else
            (Utilities|Set|Remove (Variables|Default|GetGrantedPlayers))))))
    (:else)))

(event Custom|OnPlayerPawnDied
  (params (DamagedActor :type Actor%20Object%20Reference) (Damage :type Float%20%28single-precision%29) (InstigatedBy :type Controller%20Object%20Reference) (DamageCauser :type Actor%20Object%20Reference) (HitLocation :type Vector) (FHitComponent :type Primitive%20Component%20Object%20Reference) (BoneName :type Name) (Momentum :type Vector))
  (bind _returnvalue (Team|GetActorTeam InstigatedBy))
  ;[0.4836 0.6408 1] On team elminiation send a trigger message
  (bind _pawn (Utilities|IsPlayerinRunningMinigame DamagedActor)
    (:%20%20
      (Utilities|FlowControl|Sequence
        (:then_0
          (bind _bteameliminated (CallFunction|IsTeamOutOfRespawns))
          (if (and (not (Variables|Default|GetbTeamOutOfRespawnsSignalTriggered)) _bteameliminated)
            (GameplayEvent|BroadcastEvent (Variables|UserOptions|Events|GetOnTeamIsOutOfRespawns) 0 (Utilities|Struct|MakeGameplayEventTeamEliminated 0 (Team|GetActorTeam (Pawn|GetController _pawn)) _returnvalue))
            (Variables|Default|SetbTeamOutOfRespawnsSignalTriggered true)))
        (:then_1
          (if (CallFunction|IsActoronSameTeam _pawn)
            (GameplayEvent|BroadcastEvent (Variables|UserOptions|Events|GetOnTeamMemberEliminated) 0 (Utilities|Struct|MakeGameplayEventTeamMemberEliminated (Pawn|GetController _pawn) (Team|GetActorTeam (Pawn|GetController _pawn)) _returnvalue))))
        (:then_2
          (if (and (Utilities|Enum|NotEqual(Enum) (Team|GetTeamAffiliation DamagedActor InstigatedBy) "Friendly") (CallFunction|IsActoronSameTeam InstigatedBy))
            (GameplayEvent|BroadcastEvent (Variables|UserOptions|Events|GetOnEnemyEliminatedByTeamMember) 0 (Utilities|Struct|MakeGameplayEventPlayerEliminatedbyTeamMember InstigatedBy (Team|GetActorTeam (Pawn|GetController _pawn)) _returnvalue))))))))

(event EventDestroyed
  (bind _minigame (TeamSettings|GetCurrentMinigame))
  (bind _holo_loop_cue (Variables|Default|GetHolo%20Loop%20Cue))
  ;[1 0.06933 0 0.5] Unbind events during cleanup
  (Default|UnbindEventfromOnItemAdded (Variables|Default|GetPickupItemList) (EventDispatchers|CreateEvent :!event Custom|OnAddedToInventoryList))
  (Default|UnbindEventfromOnItemListUpdated (Variables|Default|GetPickupItemList) (EventDispatchers|CreateEvent :!event Custom|OnItemListUpdated))
  (Teams|RefreshTeamChangeBinding 0 true (EventDispatchers|CreateEvent :!event Update%20Colors))
  (if (Audio|Components|Audio|IsPlaying _holo_loop_cue)
    (Audio|Components|Audio|FadeOut _holo_loop_cue 0.2 0.0 "Linear" :!class /Script/Engine.AudioComponent)
    (Utilities|IsValid _minigame
      (:Is%20Valid
        (Minigame|UnbindEventfromPlayerAddedDelegate _minigame (EventDispatchers|CreateEvent :!event Custom|OnPlayerAddedToMinigame))
        (CallFunction|RemoveAllPlayerBinds _minigame)))))

(event Custom|Bind%20and%20Handle%20Drop%20Box%20Collision
  (params (Minigame :type Fort%20Minigame%20Object%20Reference))
  ;[1 0.1082 0 0.25] Disable Collision on drop area during Edit mode, as overlaps are expensive.
  (Minigame|BindEventtoOnPlayModeChangedDelegate Minigame (delegate Custom|OnPlayModeChanged))
  (CallFunction|OnPlayModeChanged Minigame (Minigame|IsinPlayMode Minigame :!class /Script/FortniteGame.FortMinigame)))

(event Custom|OnPlayModeChanged
  (params (Minigame :type Fort%20Minigame%20Object%20Reference) (bIsInPlayMode :type Boolean))
  ;[1 0.1082 0 0.25] Disable Collision on drop area during Edit mode, as overlaps are expensive.
  (Collision|SetCollisionEnabled (Variables|Default|GetDropBox) (select bIsInPlayMode "NoCollision" "QueryOnly")))

(event OnRemovedfromMinigame(FortMinigameProgress)
  (params (Minigame :type Fort%20Minigame%20Object%20Reference))
  ;[1 0.3598 0.1485] Unbind events when removed from minigame
  (Utilities|IsValid Minigame
    (:Is%20Valid
      (Minigame|UnbindEventfromOnPawnTeleportedforMinigame Minigame (EventDispatchers|CreateEvent :!event Custom|InitialInventorySetup))
      (Minigame|UnbindEventfromPlayerAddedDelegate Minigame (EventDispatchers|CreateEvent :!event Custom|OnPlayerAddedToMinigame))
      (Minigame|UnbindEventfromPlayerRemovedDelegate Minigame (EventDispatchers|CreateEvent :!event Custom|OnPlayerRemoved))
      (TeamSettings|RemoveTeamSettingsfromMinigame Minigame)
      (Teams|RefreshTeamChangeBinding 0 true (EventDispatchers|CreateEvent :!event Update%20Colors))
      (CallFunction|RemoveAllPlayerBinds Minigame))))

(event Custom|OnAddedToInventoryList
  (params (AddedActor :type Actor%20Object%20Reference))
  (CallFunction|AddItemtoStartingInventory AddedActor)
  (CallFunction|ResaveActor))

(event Custom|OnItemListUpdated
  (if (== (ItemData|GetNumItems (Variables|Default|GetPickupItemList)) 0)
    (CallFunction|ClearItems)))

(event OnAddedtoMinigame(FortMinigameProgress)
  (params (Minigame :type Fort%20Minigame%20Object%20Reference))
  ;[0.2708 1 0.3543] Gives items if join in progress is set to TRUE
  (Utilities|IsValid Minigame
    (:Is%20Valid
      (CallFunction|SetupRespawnItemGrantBinds Minigame)
      (Minigame|BindEventtoOnPawnTeleportedforMinigame Minigame (delegate Custom|InitialInventorySetup))
      (Minigame|BindEventtoPlayerRemovedDelegate Minigame (EventDispatchers|CreateEvent :!event Custom|OnPlayerRemoved))
      (TeamSettings|AddTeamSettingstoMinigame Minigame)
      (Teams|RefreshTeamChangeBinding 0 false (EventDispatchers|CreateEvent :!event Update%20Colors))
      (CallFunction|BindAndHandleDropBoxCollision Minigame))))

(event OnComponentHit(DropBox)
  (params (HitComponent :type Primitive%20Component%20Object%20Reference) (OtherActor :type Actor%20Object%20Reference) (OtherComp :type Primitive%20Component%20Object%20Reference) (NormalImpulse :type Vector :default "0, 0, 0") (Hit :type Hit%20Result%20Structure%20%28by%20ref%29))
  ;[1 1 1] Handles item being added from a hit event.  Hit captures all pick-ups
  (Networking|SwitchHasAuthority
    (:Authority
      (bind _asfort_pickup_athena (Utilities|Casting|CastToFortPickupAthena OtherActor))
      (Utilities|IsPickupValid _asfort_pickup_athena (Variables|Default|GetPickupValidGameplayTagContainer)
        (:IsValid
          (CallFunction|AddItemtoQueue _asfort_pickup_athena))))))

(event Custom|InitialInventorySetup
  (params (PlayerPawn :type Fort%20Player%20Pawn%20Object%20Reference))
  (bind _returnvalue (Pawn|GetController PlayerPawn))
  ;[0.2708 1 0.3543] Gives items if join in progress is set to TRUE
  (bind _isinsamevolume (CallFunction|IsPlayeronSameIsland PlayerPawn))
  (if _isinsamevolume
    (bind _currentminigame (Variables|/Script/FortniteGame.FortMinigameLogicComponent|GetCurrentMinigame (Variables|Default|GetFortMinigameProgress))
      (:then
        (bind _gameisrunning (Default|MinigameIsRunning))
        (EventDispatchers|BindEventtoOnDied PlayerPawn (EventDispatchers|CreateEvent :!event Custom|OnPlayerPawnDied) :!class /Script/FortniteGame.FortPawn)
        (if (Math|Boolean|ANDBoolean _gameisrunning (Minigame|IsinProgress _currentminigame) (CallFunction|IsActoronSameTeam PlayerPawn))
          (bind _asfort_player_state_athena (Utilities|Casting|CastToFortPlayerStateAthena (Variables|/Script/Engine.Controller|GetPlayerState _returnvalue)))
          (Utilities|ShouldGrantItemstoPlayer (PlayerState|GetPlayerUniqueNetID _asfort_player_state_athena) _returnvalue
            (:true
              (bind _asfort_inventory_owner_interface (Utilities|Casting|CastToFortInventoryOwnerInterface _returnvalue))
              (CallFunction|TryGiveAllItemstoPlayer _asfort_inventory_owner_interface)
              (Utilities|Set|Add (Variables|Default|GetGrantedPlayers) (PlayerState|GetPlayerUniqueNetID _asfort_player_state_athena))))))
      (:else))))

(event EventOnMinigameEnded
  ;[0.2708 1 0.3543] Gives items if join in progress is set to TRUE
  (Networking|SwitchHasAuthority
    (:Authority
      (Minigame|UnbindEventfromPlayerChangedTeamDelegate (Utilities|GetMinigame) (delegate Custom|On%20Player%20Changed%20Team))
      (Utilities|Set|Clear (Variables|Default|GetGrantedPlayers)))))

(event EventOnMinigameStarted
  (Networking|SwitchHasAuthority
    (:Authority
      (Minigame|BindEventtoPlayerChangedTeamDelegate (Utilities|GetMinigame) (delegate Custom|On%20Player%20Changed%20Team))
      (Variables|Default|SetbTeamOutOfRespawnsSignalTriggered false))))

(event Custom|UpdateTeamAssignmentDisplay
  ;[1 1 1] Updates the display of which team the toy  is assigned to
  (Networking|CanExecuteCosmeticEvents
    (:True
      (Rendering|Material|SetScalarParameterValue (Variables|Default|GetMID%20Team%20Assignment) "Score" (select (Knobs|Team|IsTeamAll (Variables|Default|GetTeamSetting)) 0.0 (Math|Conversions|ToFloat(Integer) (Knobs|Team|GetTeamValue (Variables|Default|GetTeamSetting)))) :!class /Script/Engine.MaterialInstanceDynamic))))

(event Custom|On%20Player%20Changed%20Team
  (params (PlayerState :type Fort%20Player%20State%20Object%20Reference) (TeamId :type Byte) (OldTeamId :type Byte) (bTeamChangedWithoutRespawn :type Boolean))
  (bind _returnvalue (Class|FortPlayerState|FindFNEPlayerComponent PlayerState))
  (bind _returnvalue_2 (PlayerState|GetCurrentPawn PlayerState))
  (Utilities|Set|Remove (Variables|Default|GetGrantedPlayers) (PlayerState|GetPlayerUniqueNetID PlayerState))
  (if bTeamChangedWithoutRespawn
    ;[1 1 1] This is to support team switching without respawn otherwise this would be handled by the OnPawnTeleportedForMinigame event.
    (Utilities|IsValid _returnvalue_2
      (:Is%20Valid
        (CallFunction|InitialInventorySetup _returnvalue_2)
        (Utilities|IsValid _returnvalue
          (:Is%20Valid
            (Class|FNEPlayerComponent|ApplyMutatorstoPlayer _returnvalue false)))))))

(event Custom|OnPlayerRemoved
  (params (UniqueNetId :type Unique%20Net%20Id%20Structure) (IsLocalPlayer :type Boolean))
  (bind _returnvalue (Fortnite|Athena|GameState|GetGameStateAthena))
  (bind _returnvalue_2 (Class|FortGameState|GetFortPlayerStatefromUniqueNetId _returnvalue UniqueNetId))
  (bind _returnvalue_3 (Fortnite|GetControllerfromActor _returnvalue_2))
  (bind _returnvalue_4 (PlayerState|GetCurrentPawn _returnvalue_2))
  ;[1 0.2636 0.03237] Stop tracking players that have been removed from the minigame
  (Utilities|Set|Remove (Variables|Default|GetGrantedPlayers) UniqueNetId)
  (Development|PrintString "OnPlayerRemoved")
  (Utilities|IsValid _returnvalue
    (:Is%20Valid
      (Utilities|IsValid _returnvalue_2
        (:Is%20Valid
          (Utilities|FlowControl|Sequence
            (:then_0
              (Utilities|IsValid _returnvalue_3
                (:Is%20Valid
                  (Pawn|UnbindEventfromOnPossessedPawnChanged _returnvalue_3 (EventDispatchers|CreateEvent :!event Custom|OnPossessedPawnChanged)))))
            (:then_1
              (Utilities|IsValid _returnvalue_4
                (:Is%20Valid
                  (EventDispatchers|UnbindEventfromOnDied _returnvalue_4 (EventDispatchers|CreateEvent :!event Custom|OnPlayerPawnDied) :!class /Script/FortniteGame.FortPawn))))))))))

(event Custom|EndRoundFromEvent
  (params (Instigator :type Controller%20Object%20Reference))
  (bind _currentminigame (Variables|/Script/FortniteGame.FortMinigameLogicComponent|GetCurrentMinigame (Variables|Default|GetFortMinigameProgress))
    (:then
      (if (Knobs|Team|IsValidTeamIndex (Variables|Default|GetTeamSetting))
        (Minigame|SetWinningTeamIndex _currentminigame (Teams|IndextoTeam (Knobs|Team|GetTeamValueasIndex (Variables|Default|GetTeamSetting)))))
      (Minigame|EndGame _currentminigame "EndRound"))
    (:else)))

(event MinigameStateChangedDelegate(FortMinigameProgress)
  (params (Minigame :type Fort%20Minigame%20Object%20Reference) (NewMinigameState :type EFortMinigameState%20Enum :default "PreGame"))
  ;[1 1 1] Update team settings to make sure that they were not overwritten by another instance of the device during island editing
  (switch Utilities|FlowControl|Switch|SwitchonEFortMinigameState NewMinigameState
    (:Setup
      (TeamSettings|AddTeamSettingstoMinigame (TeamSettings|GetCurrentMinigame)))))

(event OnVisibilityChanged(VisibleInGame)
  (params (Visible :type Boolean))
  (CallFunction|UpdateSpawnerFX))

(event OnItemListUpdated(PickupItemList)
  (if (== (ItemData|GetNumItems (Variables|Default|GetPickupItemList)) 0)
    (CallFunction|ClearItems)))

(event Custom|RPC_MULTICAST%20Add%20Item%20FX
  (params (Item :type Actor%20Object%20Reference))
  (Networking|CanExecuteCosmeticEvents
    (:True
      (bind _asfort_pickup_athena (Utilities|Casting|CastToFortPickupAthena Item))
      (Class|CreativeItemPreviewComponent|ItemAddedToQueueFX (Variables|Default|GetCreative_ItemPreview_Component) _asfort_pickup_athena))))

(event Custom|Update%20Spawner%20FX
  (if (Math|Conversions|ToBoolean(Integer) (ItemData|GetNumItems (Variables|Default|GetPickupItemList)))
    (bind (_direction _trapvisuals) (Timeline|SetTrapActiveVisuals PlayFromStart)
      (:Update
        (Rendering|Material|SetVectorParameterValue (Variables|Default|GetTrapMID) "EmissiveMultiplier" (Math|Color|Lerp(LinearColor) :B (Variables|Default|GetActiveColor) :Alpha _trapvisuals) :!class /Script/Engine.MaterialInstanceDynamic)
        (Rendering|Material|SetScalarParameterValue (Variables|Default|GetRaysMID) "Fade" _trapvisuals :!class /Script/Engine.MaterialInstanceDynamic)))
    (else
      (Timeline|SetTrapActiveVisuals ReverseFromEnd))))

(fn AddItemToQueue
  (params (NewItem :type Fort%20Pickup%20Athena%20Object%20Reference))
  (bind _banyitemblacklisted (CallFunction|IsAnyItemBlacklisted (Variables|GetNewItem)))
  (if (not _banyitemblacklisted)
    (bind _bitemadded (Class|PickupItemListComponent|AddItemtoQueue (Variables|Default|GetPickupItemList) (Variables|GetNewItem)))
    (if _bitemadded
      (CallFunction|RPCMULTICASTAddItemFX (Variables|GetNewItem)))))

(fn AddItemToStartingInventory
  (params (AddedActor :type Actor%20Object%20Reference))
  (Networking|FlushNetDormancy))

(fn AddTeamSettingsToMinigame
  (params (Minigame :type Fort%20Minigame%20Object%20Reference))
  (Networking|SwitchHasAuthority
    (:Authority
      (Utilities|FlowControl|Sequence
        (:then_0
          (LocalVariables|SetCurrentTeamIndex (select (Knobs|Team|IsTeamAll (Variables|Default|GetTeamSetting)) 255 (Teams|IndextoTeam (Knobs|Team|GetTeamValueasIndex (Variables|Default|GetTeamSetting))))))
        (:then_1
          (bind _output_get (LocalVariables|SetTeamSettingsMutator (TeamSettings|GetTeamSettingsMutator (Variables|GetMinigame))))
          (Utilities|IsValid _output_get
            (:Is%20Valid
              (bind _returnvalue (TeamSettings|AddDataSourceforTeam _output_get (LocalVariables|GetCurrentTeamIndex) (Variables|Default|GetToyOptionsComponent)))
              (if _returnvalue
                (Minigame|GameRules|RefreshGameRulesPreview (Variables|GetMinigame))))))
        (:then_2
          (bind _output_get_2 (LocalVariables|SetWeaponSettingsMutator (TeamSettings|GetWeaponSettingsMutator (Variables|GetMinigame))))
          (Utilities|IsValid _output_get_2
            (:Is%20Valid
              (Class|FortAthenaMutatorWeaponSettings|AddDataSourceforTeam _output_get_2 "/Game/Athena/Items/Cosmetics/Pickaxes/DefaultPickaxe.DefaultPickaxe" (LocalVariables|GetCurrentTeamIndex) (Variables|Default|GetToyOptionsComponent)))))
        (:then_3
          _)))
    (:Remote
      _))
  (Class|FortAthenaMutatorUtilities|RemoveDataRegistryOptionsforContextObject)
  (Class|FortAthenaMutatorUtilities|ApplyActorOptionsComponenttoDataRegistry (Variables|Default|GetToyOptionsComponent)))

(fn ApplyColors
  (params (Color :type Linear%20Color%20Structure))
  (bind _team_color_emissive_intensity (Variables|Default|GetTeam%20Color%20Emissive%20Intensity))
  (bind _returnvalue (Math|Color|LinearColor*Float Color _team_color_emissive_intensity))
  (Rendering|Material|SetVectorParameterValue (Material|GetMID (Variables|Default|GetSm_BotTurret_ProjectileRing_01)) "ColorOutline" _returnvalue :!class /Script/Engine.MaterialInstanceDynamic)
  (Rendering|Material|SetVectorParameterValue (Variables|Default|GetMID%20Team%20Assignment) "Color" _returnvalue :!class /Script/Engine.MaterialInstanceDynamic) ; Set team color to match if we ever have assigned colors for each team
  (Rendering|Material|SetVectorParameterValue (Variables|Default|GetRaysMID) "ColorHigh" _returnvalue :!class /Script/Engine.MaterialInstanceDynamic)
  (Rendering|Material|SetVectorParameterValue (Variables|Default|GetRaysMID) "ColorLow" (Variables|GetColor) :!class /Script/Engine.MaterialInstanceDynamic)
  (Variables|Default|SetActiveColor (Variables|GetColor))
  (CallFunction|UpdateSpawnerFX))

(fn BuildDataRegistryResolverScope
  (bind _teamsetting (Variables|Default|GetTeamSetting))
  (Utilities|FlowControl|Sequence
    (:then_0
      (bind _output_get (LocalVariables|SetSpatialInterfaceName (SpatialGameplay|GetSpatialGameplayInterfaceFNameforActor self)))
      (if (!= _output_get (LocalVariables|GetNameNone))
        (Utilities|Array|Add (LocalVariables|GetOutResolverScopes) _output_get)))
    (:then_1
      (if (not (Knobs|Team|IsTeamAll _teamsetting))
        (Utilities|Array|Add (LocalVariables|GetOutResolverScopes) (Utilities|String|StringToName (Utilities|String|BuildString(Integer) "Team" "" (Math|Conversions|ToInteger(Byte) (Teams|IndextoTeam (Knobs|Team|GetTeamValueasIndex _teamsetting))))))))
    (:then_2
      (return (LocalVariables|GetOutResolverScopes) 0 (> (Utilities|Array|Length (LocalVariables|GetOutResolverScopes)) 0)))))

(fn ClearItems
  (Networking|SwitchHasAuthority
    (:Authority
      (Networking|FlushNetDormancy)
      (CallFunction|ResaveActor))))

(fn GetCacheContextOverride
  (return 0))

(fn GetCurrentMinigame
  (return (Variables|/Script/FortniteGame.FortMinigameLogicComponent|GetCurrentMinigame (Variables|Default|GetFortMinigameProgress))))

(fn GetTeamSettingsMutator
  (params (Minigame :type Fort%20Minigame%20Object%20Reference))
  (Utilities|FlowControl|Sequence
    (:then_0
      (bind _output_get (LocalVariables|SetCurrentMinigame Minigame))
      (Utilities|IsValid _output_get
        (:Is%20Valid
          (bind _output_get_2 (LocalVariables|SetMutatorList (Actor|GetComponentbyClass _output_get "/Script/FortniteGame.FortMutatorListComponent")))
          (Utilities|IsValid _output_get_2
            (:Is%20Valid
              (bind _asfort_athena_mutator_player_team_settings (Utilities|Casting|CastToFortAthenaMutator_PlayerTeamSettings (Modifier|GetMutatorbyClass _output_get_2 "/Script/FortniteGame.FortAthenaMutator_PlayerTeamSettings")))
              (LocalVariables|SetTeamSettingsMutator _asfort_athena_mutator_player_team_settings))))))
    (:then_1
      (return (LocalVariables|GetTeamSettingsMutator)))))

(fn GetVisualComponents
  (return (Utilities|Array|MakeArray (Variables|Default|GetSm_BotTurret_ProjectileRing_01) (Variables|Debug|GetStaticMeshComponent) (Variables|Default|GetTeamNumberDisplay) (Variables|Default|GetHolo%20Loop%20Cue) (Variables|Default|GetLightRays) (Variables|Default|GetPreviewStaticMesh) (Variables|Default|GetPreviewSkeletalMesh))))

(fn GetWeaponSettingsMutator
  (params (Minigame :type Fort%20Minigame%20Object%20Reference))
  (bind _minigame (Variables|GetMinigame))
  (bind _volume (Variables|/Script/FortniteGame.FortMinigame|GetVolume _minigame))
  (Utilities|FlowControl|Sequence
    (:then_0
      (Utilities|IsValid _minigame
        (:Is%20Valid
          (Utilities|IsValid _volume
            (:Is%20Valid
              (bind _output_get (LocalVariables|SetCreativeIslandMutatorList (Actor|GetComponentbyClass _volume "/Game/Athena/Playlists/Creative/Mutators/CreativeIslandMutatorList.CreativeIslandMutatorList_C")))
              (Utilities|IsValid _output_get
                (:Is%20Valid
                  (bind _asfort_athena_mutator_weapon_settings (Utilities|Casting|CastToFortAthenaMutator_WeaponSettings (Modifier|GetMutatorbyClass _output_get "/Script/FortniteGame.FortAthenaMutator_WeaponSettings")))
                  (LocalVariables|SetOutMutator _asfort_athena_mutator_weapon_settings))))))))
    (:then_1
      (return (LocalVariables|GetOutMutator)))))

(fn GiveAmmo
  (params (AmmoDefinition :type Fort%20World%20Item%20Definition%20Object%20Reference) (InventoryOwner :type Fort%20Inventory%20Owner%20Interface%20Interface))
  (bind _returnvalue (WeaponItemDefinition|GetAmmoWorldItemDefinition AmmoDefinition))
  (if (and (Variables|Default|GetGrantExtraAmmoSetting) (Utilities|NotEqual(Object) AmmoDefinition _returnvalue))
    (if (== (Inventory|GetItemQuantityonInventoryOwner InventoryOwner _returnvalue) 0)
      ;[1 1 1] Giving out ammo
      (if (Utilities|Equal(Object) _returnvalue "/Game/Athena/Items/Ammo/AthenaAmmoDataBulletsHeavy.AthenaAmmoDataBulletsHeavy")
        (Inventory|GiveItemtoInventoryOwner :InventoryOwner InventoryOwner :ItemDefinition _returnvalue :NumberToGive 10 :bNotifyPlayer true :ItemLevel -1 :PickupInstigatorHandle -1 :bUseItemPickupAnalyticEvent false :WeaponAmmoOverride -1)
        (elif (Utilities|Equal(Object) _returnvalue "/Game/Athena/Items/Ammo/AthenaAmmoDataBulletsLight.AthenaAmmoDataBulletsLight")
          (Inventory|GiveItemtoInventoryOwner :InventoryOwner InventoryOwner :ItemDefinition _returnvalue :NumberToGive 50 :bNotifyPlayer true :ItemLevel -1 :PickupInstigatorHandle -1 :bUseItemPickupAnalyticEvent false :WeaponAmmoOverride -1)
          (elif (Utilities|Equal(Object) _returnvalue "/Game/Athena/Items/Ammo/AthenaAmmoDataShells.AthenaAmmoDataShells")
            (Inventory|GiveItemtoInventoryOwner :InventoryOwner InventoryOwner :ItemDefinition _returnvalue :NumberToGive 30 :bNotifyPlayer true :ItemLevel -1 :PickupInstigatorHandle -1 :bUseItemPickupAnalyticEvent false :WeaponAmmoOverride -1)
            (elif (Utilities|Equal(Object) _returnvalue "/Game/Athena/Items/Ammo/AmmoDataRockets.AmmoDataRockets")
              (Inventory|GiveItemtoInventoryOwner :InventoryOwner InventoryOwner :ItemDefinition _returnvalue :NumberToGive 10 :bNotifyPlayer true :ItemLevel -1 :PickupInstigatorHandle -1 :bUseItemPickupAnalyticEvent false :WeaponAmmoOverride -1)
              (elif (Utilities|Equal(Object) _returnvalue "/Game/Athena/Items/Ammo/AthenaAmmoDataBulletsMedium.AthenaAmmoDataBulletsMedium")
                (Inventory|GiveItemtoInventoryOwner :InventoryOwner InventoryOwner :ItemDefinition _returnvalue :NumberToGive 30 :bNotifyPlayer true :ItemLevel -1 :PickupInstigatorHandle -1 :bUseItemPickupAnalyticEvent false :WeaponAmmoOverride -1)
                (else
                  (Inventory|GiveItemtoInventoryOwner :InventoryOwner InventoryOwner :ItemDefinition _returnvalue :NumberToGive 5 :bNotifyPlayer true :ItemLevel -1 :PickupInstigatorHandle -1 :bUseItemPickupAnalyticEvent false :WeaponAmmoOverride -1))))))))))

(fn InitializeMIDs
  (Networking|CanExecuteCosmeticEvents
    (:True
      (Utilities|FlowControl|Sequence
        (:then_0
          (Variables|Default|GetMID%20Team%20Assignment
            (:then)
            (:else
              (bind _returnvalue (Rendering|Material|CreateDynamicMaterialInstance (Variables|Default|GetTeamNumberDisplay) 0 (Variables|Default|GetTeamAssignmentMat) "None" :!class /Script/Engine.PrimitiveComponent))
              (Variables|Default|SetMID%20Team%20Assignment _returnvalue))))
        (:then_1
          (Variables|Default|GetRaysMID
            (:then)
            (:else
              (bind _returnvalue_2 (Rendering|Material|CreateDynamicMaterialInstance (Variables|Default|GetLightRays) 0 0 "None" :!class /Script/Engine.PrimitiveComponent))
              (Variables|Default|SetRaysMID _returnvalue_2))))
        (:then_2
          (Variables|Default|GetTrapMID
            (:then)
            (:else
              (bind _returnvalue_3 (Rendering|Material|CreateDynamicMaterialInstance (Variables|Debug|GetStaticMeshComponent) 0 0 "None" :!class /Script/Engine.PrimitiveComponent))
              (Variables|Default|SetTrapMID _returnvalue_3))))))))

(fn IsVisibleDuringMinigame
  (return (Variables|Default|GetIsMeshVisibleDuringMinigame)))

(fn IsActorOnSameTeam
  (params (Actor :type Actor%20Object%20Reference))
  (return (or (Knobs|Team|IsTeamAll (Variables|Default|GetTeamSetting)) (== (Teams|TeamtoIndex (Team|GetActorTeam Actor)) (Knobs|Team|GetTeamValueasIndex (Variables|Default|GetTeamSetting))))))

(fn IsAnyItemBlacklisted
  (params (InPickup :type Fort%20Pickup%20Object%20Reference))
  (bind _index (Utilities|FlowControl|ForLoopwithBreak 0 (- (Pickup|GetItemEntryNum InPickup) 1) :!label _label)
    (:LoopBody
      (Utilities|IsBlacklisted (Class|FortItemFunctionLibrary|BreakFortItemEntry (Pickup|GetItemEntry (Variables|GetInPickup) _index))
        (:Blacklisted
          (LocalVariables|SetbWasBlacklisted true)
          (goto _label Break))))
    (:Completed
      (return (LocalVariables|GetbWasBlacklisted)))))

(fn IsPlayerOnSameIsland
  (params (Actor :type Actor%20Object%20Reference))
  (Utilities|IsActorRelevanttoMe Actor
    (:IsRelevant
      (return true))
    (:NotRelevant
      (return false))))

(fn IsTeamOutOfRespawns
  (bind _outplayers (Minigame|GetParticipatingPlayers_NotPure (Variables|/Script/FortniteGame.FortMinigameLogicComponent|GetCurrentMinigame (Variables|Default|GetFortMinigameProgress))))
  (bind _array_element (Utilities|Array|ForEachLoopwithBreak _outplayers :!label _label)
    (:LoopBody
      (if (LocalVariables|GetbTeamIsOutOfRespawns)
        ;[1 1 1] Body loop, determine if the player is on our team and if they have lives left
        (Utilities|IsValid _array_element
          (:Is%20Valid
            (bind _isinsamevolume (CallFunction|IsPlayeronSameIsland _array_element))
            (if (and _isinsamevolume (CallFunction|IsActoronSameTeam _array_element))
              (LocalVariables|SetTeamSize (Math|Byte|Byte+Byte (LocalVariables|GetTeamSize) 1))
              (Utilities|IsValid _array_element
                (:Is%20Valid
                  (Utilities|IsValid (Class|FortPlayerState|FindFNEPlayerComponent _array_element)
                    (:Is%20Valid
                      (if (not (Class|FNEPlayerComponent|CanPlayerRespawn (Class|FortPlayerState|FindFNEPlayerComponent _array_element)))
                        (Utilities|IsValid (PlayerState|GetCurrentPawn _array_element)
                          (:Is%20Valid
                            (if (FortPawn|IsDead (PlayerState|GetCurrentPawn _array_element))
                              _
                              (else
                                _)))))
                      (LocalVariables|SetbTeamIsOutOfRespawns false))))))))
        (else
          (goto _label Break))))
    (:Completed
      (return (and (LocalVariables|GetbTeamIsOutOfRespawns) (Math|Byte|Byte>Byte (LocalVariables|GetTeamSize) 0))))))

(fn MigrateDataToPickupList
  (bind _tospawnlist (Variables|Default|GetToSpawnList))
  (bind _returnvalue (Utilities|Array|Length _tospawnlist))
  (bind _returnvalue_2 (> _returnvalue 0))
  (bind _tospawnlist_2 (Variables|Default|GetToSpawnList))
  ;[1 1 1] This function is here to fill up PickupItemList Component on islands previous this device refactor on 12.30.
  ; After the refactor, PickupItemList is in charge of items but those were previously stored on ToSpawnList.
  ; If ToSpawnList contains items and PickupItemList doesn't, then, fill the list with already saved data
  (if (and _returnvalue_2 (== (ItemData|GetNumItems (Variables|Default|GetPickupItemList)) 0))
    (ItemData|AddItemsfromSpawnParams (Variables|Default|GetPickupItemList) _tospawnlist_2)
    (elif _returnvalue_2
      _))
  (Utilities|Array|Clear _tospawnlist_2))

(fn OverrideCollisionSettings
  (bind _staticmeshcomponent (Variables|Debug|GetStaticMeshComponent))
  ;[1 0.9605 0.1332] Something is preventing the collision settings from actually sticking. Force them to fit what they should be
  (Collision|SetCollisionResponsetoChannel _staticmeshcomponent "ECC_Visibility" "ECR_Block")
  (Collision|SetCollisionResponsetoChannel _staticmeshcomponent "ECC_GameTraceChannel2" "ECR_Overlap")
  (Collision|SetCollisionResponsetoChannel _staticmeshcomponent "ECC_GameTraceChannel11" "ECR_Block")
  (Collision|SetCollisionResponsetoChannel _staticmeshcomponent "ECC_GameTraceChannel16" "ECR_Overlap"))

(fn PropertiesUpdated
  (|CallParentFunction)
  (TeamSettings|AddTeamSettingstoMinigame (TeamSettings|GetCurrentMinigame))
  (CallFunction|InitializeMIDs)
  (CallFunction|UpdateSpawnerFX)
  (CallFunction|UpdateTeamAssignmentDisplay)
  (DeviceVisuals|UpdateColors))

(fn RemoveAllPlayerBinds
  (params (Minigame :type Fort%20Minigame%20Object%20Reference))
  ;[1 0.298 0.1277] Unbinds bound events on all players
  (Utilities|FlowControl|Sequence
    (:then_0
      (Utilities|IsValid Minigame
        (:Is%20Valid
          (for _array_element (Minigame|GetParticipatingPlayers Minigame)
            (Utilities|IsValid _array_element
              (:Is%20Valid
                (Utilities|IsValid (Fortnite|GetControllerfromActor _array_element)
                  (:Is%20Valid
                    (Pawn|UnbindEventfromOnPossessedPawnChanged (Fortnite|GetControllerfromActor _array_element) (EventDispatchers|CreateEvent :!event OnPossessedPawnChanged))))))))))
    (:then_1
      (bind _outfortplayerpawns (Fortnite|GetFortPlayerPawns))
      (for _array_element_2 _outfortplayerpawns
        (Utilities|IsValid _array_element_2
          (:Is%20Valid
            (EventDispatchers|UnbindEventfromOnDied _array_element_2 (EventDispatchers|CreateEvent :!event OnPlayerPawnDied) :!class /Script/FortniteGame.FortPawn)))))))

(fn RemoveTeamSettingsFromMinigame
  (params (Minigame :type Fort%20Minigame%20Object%20Reference))
  (Utilities|FlowControl|Sequence
    (:then_0
      (bind _output_get (LocalVariables|SetTeamSettingsMutator (TeamSettings|GetTeamSettingsMutator (Variables|GetMinigame))))
      (Utilities|IsValid _output_get
        (:Is%20Valid
          (bind _returnvalue (TeamSettings|RemoveDataSource _output_get (Variables|Default|GetToyOptionsComponent)))
          (if _returnvalue
            (Minigame|GameRules|RefreshGameRulesPreview (Variables|GetMinigame))))))
    (:then_1
      (bind _output_get_2 (LocalVariables|SetWeaponSettingsMutator (TeamSettings|GetWeaponSettingsMutator (Variables|GetMinigame))))
      (Utilities|IsValid _output_get_2
        (:Is%20Valid
          (Class|FortAthenaMutatorWeaponSettings|RemoveDataSource _output_get_2 (Variables|Default|GetToyOptionsComponent)))))
    (:then_2
      (Class|FortAthenaMutatorUtilities|RemoveDataRegistryOptionsforContextObject))))

(fn ResaveActor
  (CallFunction|UpdateTeamAssignmentDisplay)
  (CallFunction|Resave))

(fn TryGiveAllItemsToPlayer
  (params (PlayerToGiveItemsTo :type Fort%20Inventory%20Owner%20Interface%20Interface))
  (bind _returnvalue (Math|Transform|MakeTransform "0, 0, 0"))
  (bind _pickupitemlist (Variables|Default|GetPickupItemList))
  (bind _returnvalue_2 (ItemData|MakeSpawnParamsList _pickupitemlist _returnvalue))
  (bind _playertogiveitemsto (Variables|GetPlayerToGiveItemsTo))
  (bind _asfort_player_controller_athena (Utilities|Casting|CastToFortPlayerControllerAthena PlayerToGiveItemsTo)
    (:then
      (LocalVariables|SetPlayerController _asfort_player_controller_athena))
    (:CastFailed
      _))
  (Utilities|FlowControl|Sequence
    (:then_0
      (bind _playercontroller (LocalVariables|GetPlayerController)
        (:then
          (switch Utilities|FlowControl|Switch|SwitchonE_TeamSettings_GrantBehavior (Variables|Default|GetGrantBehavior)
            (:NewEnumerator0
              (Class|FortPlayerControllerAthena|ClearDroppableItems _playercontroller false) ; Clear inventory before giving items.  Avoids duplicates on respawn
              (Variables|/Script/FortniteGame.FortPlayerControllerAthena|SetbShouldForceDeleteDroppedItems true _playercontroller))
            (:NewEnumerator1
              (Class|FortPlayerControllerAthena|RemoveDroppableItems _playercontroller)
              (Variables|/Script/FortniteGame.FortPlayerControllerAthena|SetbShouldForceDeleteDroppedItems true _playercontroller))))
        (:else)))
    (:then_1
      (bind _outvolumemanager (Volume|GetVolumeManager)
        (:Valid
          (LocalVariables|SetVolumeManager _outvolumemanager))))
    (:then_2
      (for (_array_element _array_index) _returnvalue_2
        (if (Math|Boolean|ANDBoolean (Utilities|IsValid (Utilities|ResolveSoftReference (Utilities|Struct|BreakMinigameSpawnerSpawnParams _array_element)) :!class /Script/Engine.KismetSystemLibrary) (not (Validation|IsBlockListedItemDefinition (Utilities|ResolveSoftReference (Utilities|Struct|BreakMinigameSpawnerSpawnParams _array_element)))) (Validation|IsItemDefinitionAllowed (Utilities|ResolveSoftReference (Utilities|Struct|BreakMinigameSpawnerSpawnParams _array_element))))
          (Inventory|GiveItemtoInventoryOwner _playertogiveitemsto (Utilities|ResolveSoftReference (Utilities|Struct|BreakMinigameSpawnerSpawnParams _array_element)) (.ItemVariantGuid (Utilities|Struct|BreakMinigameSpawnerSpawnParams _array_element)) (.PickupQuantity (Utilities|Struct|BreakMinigameSpawnerSpawnParams _array_element)))
          (CallFunction|GiveAmmo (Utilities|ResolveSoftReference (Utilities|Struct|BreakMinigameSpawnerSpawnParams _array_element)) _playertogiveitemsto)
          (if (== (Math|Integer|Min(Integer) (Utilities|Array|LastIndex _returnvalue_2) (Variables|Default|GetEquipItemOnGrant)) _array_index)
            (bind _playercontroller_2 (LocalVariables|GetPlayerController)
              (:then
                (Inventory|ExecuteInventoryItemDefinition :self _playercontroller_2 :ItemDefinition (Utilities|ResolveSoftReference (Utilities|Struct|BreakMinigameSpawnerSpawnParams _array_element)) :Delay 0.0 :bForceExecute true :bActivateSlotAfterSettingFocused true))
              (:else)))))
      (bind _playercontroller_3 (LocalVariables|GetPlayerController)
        (:then
          (Variables|/Script/FortniteGame.FortPlayerControllerAthena|SetbShouldForceDeleteDroppedItems false _playercontroller_3))
        (:else)))))

(fn UpdateColors
  (bind _teamsetting (Variables|Default|GetTeamSetting))
  (Networking|CanExecuteCosmeticEvents
    (:True
      (Utilities|FlowControl|Sequence
        (:then_0
          (bind _fortminigameprogress (Variables|Default|GetFortMinigameProgress)
            (:then
              (bind _currentminigame (Variables|/Script/FortniteGame.FortMinigameLogicComponent|GetCurrentMinigame _fortminigameprogress)
                (:then
                  (if (Knobs|Team|IsTeamAll (Variables|Default|GetTeamSetting))
                    (LocalVariables|Setcolor%20Index -1)
                    (else
                      (Utilities|IsValid (Minigame|GetMinigameTeamComponent _currentminigame)
                        (:Is%20Valid
                          (LocalVariables|Setcolor%20Index (Minigame|GameRules|GetTeamColorIndex (Minigame|GetMinigameTeamComponent _currentminigame) (Teams|IndextoTeam (Knobs|Team|GetTeamValueasIndex (Variables|Default|GetTeamSetting))))))
                        (:Is%20Not%20Valid
                          _)))))
                (:else
                  _))
              ;[0.8827 0.1953 1 0.25] Editor Logic, ignore minigame, use team value directly.
              (if (Editor|InEditor)
                (if (Knobs|Team|IsTeamAll _teamsetting)
                  (LocalVariables|Setcolor%20Index -1)
                  (else
                    _)))
              (LocalVariables|Setcolor%20Index (Math|Integer|Max(Integer) (Math|Integer|%25%28Integer%29 (Knobs|Team|GetTeamValue _teamsetting) 16))))
            (:else)))
        (:then_1
          (bind _result (ArrayMacros|SafeArrayGet (Class|GetClassDefaults "/Game/Creative/Sets/MilitaryBase/Props/CreativeTeamColors.CreativeTeamColors_C") (select (== (LocalVariables|Getcolor%20Index) -1) (select (Knobs|Team|IsTeamAll (Variables|Default|GetTeamSetting)) 0 (Knobs|Team|GetTeamValue (Variables|Default|GetTeamSetting))) (LocalVariables|Getcolor%20Index)))
            (:%20%20
              (DeviceVisuals|ApplyColors _result))))))))

(fn ConstructionScript
  (Networking|CanExecuteCosmeticEvents
    (:True
      (for _ (range 1 (+ (Variables|Default|GetNumOfClientMIDsToCreate) 1))
        (bind _returnvalue (Rendering|Material|CreateDynamicMaterialInstance "/Game/Packages/Fortnite_UniqueMaterials_1/BackPackShrinkingEffect.BackPackShrinkingEffect" "None" "None" :!class /Script/Engine.KismetMaterialLibrary))
        (Utilities|Array|Add (Variables|Default|GetPreviewMeshMID) _returnvalue)))))
